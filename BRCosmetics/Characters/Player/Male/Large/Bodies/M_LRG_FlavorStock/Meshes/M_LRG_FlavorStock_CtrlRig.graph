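# Control Rig for the large male FlavorStock player body that applies procedural corrective bone deformation.
# Forward solve runs the shared FN_DeformRig_FNC function, then wrist_function, which offsets right wrist helper bones by remapped hand_r rotation.
# When isFrontEnd is true, neckSetup also shifts left and right neck support bones using curve-mapped head-to-neck swing.
# Several nodes, including head_ctrl-to-head and a wristband constraint, are unconnected.
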ControlRigBlueprint /BRCosmetics/Characters/Player/Male/Large/Bodies/M_LRG_FlavorStock/Meshes/M_LRG_FlavorStock_CtrlRig
graphs [
 New Function_ContainedGraph {nodes [
   RigVMFunctionEntryNode()
   > &RigUnit_ModifyTransforms(
    ItemToModify=((Item=(Type=Bone,Name="None"),Transform=(Rotation=(X=0,Y=0,Z=-0.0872,W=0.9962),Translation=(X=-1,Y=-1,Z=0),Scale3D=(X=0.9,Y=1,Z=1))))
    WeightMinimum=0
    Weight=RigVMDispatchNode(BoneName=root, Default=0, Name=deformRig_r_wrist_up_alpha) {template="DISPATCH_RigDispatch_GetAnimAttribute(in Name,in BoneName,in Default,out Value,out Found)"}.Value
   )
   > &RigUnit_ModifyTransforms(
    ItemToModify=((Item=(Type=Bone,Name="None"),Transform=(Rotation=(X=0,Y=0,Z=-0.0872,W=0.9962),Translation=(X=-0.5,Y=0.5,Z=0),Scale3D=(X=0.9,Y=1,Z=1))))
    WeightMinimum=0
    Weight=RigVMDispatchNode(BoneName=root, Default=0, Name=deformRig_r_wrist_dn_alpha) {template="DISPATCH_RigDispatch_GetAnimAttribute(in Name,in BoneName,in Default,out Value,out Found)"}.Value
   )
   > &RigUnit_ModifyTransforms(
    ItemToModify=((Item=(Type=Bone,Name="None"),Transform=(Rotation=(X=0,Y=0.0436,Z=0,W=0.999),Translation=(X=0.1,Y=0,Z=0.5),Scale3D=(X=0.7,Y=1,Z=1))))
    Weight=1
    WeightMinimum=0
   )
   > &RigUnit_ModifyTransforms(
    ItemToModify=((Item=(Type=Bone,Name="None"),Transform=(Rotation=(X=0,Y=-0.0436,Z=0,W=0.999),Translation=(X=-0.5,Y=0,Z=-0.5),Scale3D=(X=0.7,Y=1,Z=1))))
    Weight=1
    WeightMinimum=0
   )
   n1 = &RigVMFunction_MathQuaternionToEuler(
    RotationOrder=ZYX
    Value=&RigVMFunction_MathTransformMakeRelative(
     Global=&RigUnit_GetTransform(Item=(Type=Bone,Name="hand_r"), Space=LocalSpace).Transform
     Parent=&RigUnit_GetTransform(Item=(Type=Bone,Name="hand_r"), Space=LocalSpace, bInitial=true).Transform
    ).Local.Rotation
   )
   > &RigUnit_ModifyTransforms(
    ItemToModify=((Item=(Type=Bone,Name="base_wristhand_outside_r"),Transform=(Rotation=(X=0,Y=0,Z=0,W=1),Translation=(X=-25,Y=0,Z=-18),Scale3D=(X=1,Y=1,Z=1))))
    WeightMinimum=0
    Weight=&RigVMFunction_MathDoubleRemap(SourceMaximum=-52, SourceMinimum=-5, TargetMaximum=1, TargetMinimum=0, Value=@n1.Result.Y).Result
   )
   > &RigUnit_ModifyTransforms(
    ItemToModify=((Item=(Type=Bone,Name="base_wristhand_inside_r"),Transform=(Rotation=(X=0,Y=0,Z=0,W=1),Translation=(X=-8,Y=0,Z=6),Scale3D=(X=1,Y=1,Z=1))))
    WeightMinimum=0
    Weight=&RigVMFunction_MathDoubleRemap(SourceMaximum=54, SourceMinimum=35, TargetMaximum=1, TargetMinimum=0, Value=@n1.Result.Y).Result
   )
   > RigVMFunctionReturnNode()
   n2 = &RigVMFunction_MathQuaternionToEuler(
    RotationOrder=ZYX
    Value=&RigVMFunction_MathTransformMakeRelative(
     Global=&RigUnit_GetTransform(Item=(Type=Bone,Name="hand_r"), Space=LocalSpace).Transform
     Parent=&RigUnit_GetTransform(Item=(Type=Bone,Name="hand_r"), Space=LocalSpace, bInitial=true).Transform
    ).Local.Rotation
   )
   &RigUnit_ModifyTransforms(
    ExecutePin=()
    ItemToModify=((Item=(Type=Bone,Name="base_wristhand_outside_r"),Transform=(Rotation=(X=0,Y=0,Z=0,W=1),Translation=(X=-25,Y=0,Z=-18),Scale3D=(X=1,Y=1,Z=1))))
    WeightMinimum=0
    Weight=&RigVMFunction_MathDoubleRemap(SourceMaximum=-52, SourceMinimum=-25, TargetMaximum=1, TargetMinimum=0, Value=@n2.Result.Y).Result
   )
   > &RigUnit_ModifyTransforms(
    ItemToModify=((Item=(Type=Bone,Name="base_wristhand_inside_r"),Transform=(Rotation=(X=0,Y=0,Z=0,W=1),Translation=(X=-7,Y=0,Z=9),Scale3D=(X=1,Y=1,Z=1))))
    WeightMinimum=0
    Weight=&RigVMFunction_MathDoubleRemap(SourceMaximum=54, SourceMinimum=35, TargetMaximum=1, TargetMinimum=0, Value=@n2.Result.Y).Result
   )
   &RigUnit_ParentConstraint(
    AdvancedSettings=(InterpolationType=Average,RotationOrderForFilter=XZY)
    Child=(Type=Bone,Name="dyn_wristband_r")
    ExecutePin=()
    Parents=((Item=(Type=Bone,Name="lowerarm_r"),Weight=0.5),(Item=(Type=Bone,Name="hand_r"),Weight=0.5))
    Weight=1
   )
  ]}
 New Function_ContainedGraph {nodes [
   RigVMFunctionEntryNode()
   > RigVMFunction_Sequence = RigVMAggregateNode()
   n1 = &RigUnit_GetRelativeTransformForItem(Child=(Type=Bone,Name="neck_01"), Parent=(Type=Bone,Name="head"), bChildInitial=False, bParentInitial=False)
   RigVMVariableNode(
    Value=&RigVMFunction_MathQuaternionToEuler(
     RotationOrder=ZYX
     Value=&RigVMFunction_MathQuaternionSwingTwist(TwistAxis=(X=1,Y=0,Z=0), Input=@n1.RelativeTransform.Rotation).Swing
    ).Result
    ExecuteContext=@RigVMFunction_Sequence.A
   ) {variable=Head Rotation}
   n2 = RigVMVariableNode() {variable=Head Rotation}
   n3 = &RigVMFunction_AnimEvalRichCurve(
    Curve="(EditorCurveData=(Keys=((InterpMode=RCIM_Cubic,TangentMode=RCTM_User,TangentWeightMode=RCTWM_WeightedNone,Time=0,Value=0,ArriveTangent=0.2796,ArriveTangentWeight=0,LeaveTangent=0.2796,LeaveTangentWeight=0),(InterpMode=RCIM_Cubic,TangentMode=RCTM_User,Tange ...(truncated)"
    SourceMaximum=-140
    SourceMinimum=0
    TargetMaximum=-44
    TargetMinimum=0
    Value=@n2.Value.X
   )
   n4 = &RigVMFunction_AnimEvalRichCurve(
    Curve="(EditorCurveData=(Keys=((InterpMode=RCIM_Cubic,TangentMode=RCTM_User,TangentWeightMode=RCTWM_WeightedNone,Time=0,Value=0,ArriveTangent=0.2796,ArriveTangentWeight=0,LeaveTangent=0.2796,LeaveTangentWeight=0),(InterpMode=RCIM_Cubic,TangentMode=RCTM_User,Tange ...(truncated)"
    SourceMaximum=-140
    SourceMinimum=0
    TargetMaximum=-44
    TargetMinimum=0
    Value=@n2.Value.Y
   )
   n5 = &RigVMFunction_AlphaInterp(
    InterpSpeedDecreasing=10
    InterpSpeedIncreasing=10
    Scale=1
    bClampResult=false
    bInterpResult=false
    bMapRange=true
    Value=&RigVMFunction_MathDoubleAdd(A=@n3.Result, B=@n4.Result).Result
   )
   &RigUnit_ModifyTransforms(
    ItemToModify=((Item=(Type=Bone,Name="neck_support_l"),Transform=(Rotation=(X=0,Y=0,Z=0,W=1),Translation=(X=0,Y=0,Z=-5),Scale3D=(X=1,Y=1,Z=1))))
    WeightMinimum=0
    Weight=&RigVMFunction_MathDoubleMul(B=-1, A=@n5.Result).Result
    ExecutePin=@RigVMFunction_Sequence.B
   )
   n6 = RigVMVariableNode() {variable=Head Rotation}
   n7 = &RigVMFunction_AnimEvalRichCurve(
    Curve="(EditorCurveData=(Keys=((InterpMode=RCIM_Cubic,TangentMode=RCTM_User,TangentWeightMode=RCTWM_WeightedNone,Time=0,Value=0,ArriveTangent=0.2796,ArriveTangentWeight=0,LeaveTangent=0.2796,LeaveTangentWeight=0),(InterpMode=RCIM_Cubic,TangentMode=RCTM_User,Tange ...(truncated)"
    SourceMaximum=-140
    SourceMinimum=0
    TargetMaximum=-44
    TargetMinimum=0
    Value=@n6.Value.X
   )
   n8 = &RigVMFunction_AnimEvalRichCurve(
    Curve="(EditorCurveData=(Keys=((InterpMode=RCIM_Cubic,TangentMode=RCTM_User,TangentWeightMode=RCTWM_WeightedNone,Time=0,Value=0,ArriveTangent=0.2796,ArriveTangentWeight=0,LeaveTangent=0.2796,LeaveTangentWeight=0),(InterpMode=RCIM_Cubic,TangentMode=RCTM_User,Tange ...(truncated)"
    SourceMaximum=140
    SourceMinimum=0
    TargetMaximum=44
    TargetMinimum=0
    Value=@n6.Value.Y
   )
   n9 = &RigVMFunction_AlphaInterp(
    InterpSpeedDecreasing=10
    InterpSpeedIncreasing=10
    Scale=1
    bClampResult=false
    bInterpResult=false
    bMapRange=true
    Value=&RigVMFunction_MathDoubleAdd(A=@n7.Result, B=@n8.Result).Result
   )
   Multiply = &RigVMFunction_MathDoubleMul(B=1, A=@n9.Result)
   > &RigUnit_ModifyTransforms(
    ItemToModify=((Item=(Type=Bone,Name="neck_support_r"),Transform=(Rotation=(X=0,Y=0,Z=0,W=1),Translation=(X=0,Y=0,Z=5),Scale3D=(X=1,Y=1,Z=1))))
    WeightMinimum=0
    Weight=@Multiply.Result
   )
   RigVMFunctionReturnNode(ExecuteContext=@RigVMFunction_Sequence.C)
   &RigVMFunction_AnimEvalRichCurve(
    Curve="(EditorCurveData=(Keys=((InterpMode=RCIM_Cubic,TangentMode=RCTM_User,TangentWeightMode=RCTWM_WeightedNone,Time=0,Value=0,ArriveTangent=0.2796,ArriveTangentWeight=0,LeaveTangent=0.2796,LeaveTangentWeight=0),(InterpMode=RCIM_Cubic,TangentMode=RCTM_User,Tange ...(truncated)"
    SourceMaximum=-140
    SourceMinimum=0
    TargetMaximum=-44
    TargetMinimum=0
    Value=0
   )
   &RigVMFunction_AnimEvalRichCurve(
    Curve="(EditorCurveData=(Keys=((InterpMode=RCIM_Cubic,TangentMode=RCTM_User,TangentWeightMode=RCTWM_WeightedNone,Time=0,Value=0,ArriveTangent=0.2796,ArriveTangentWeight=0,LeaveTangent=0.2796,LeaveTangentWeight=0),(InterpMode=RCIM_Cubic,TangentMode=RCTM_User,Tange ...(truncated)"
    SourceMaximum=-140
    SourceMinimum=0
    TargetMaximum=-44
    TargetMinimum=0
    Value=0
   )
  ]}
 RigVMFunctionLibrary {nodes [
   neckSetup = RigVMCollapseNode(ExecuteContext=())
   wrist_function = RigVMCollapseNode(ExecuteContext=())
  ]}
 RigVMFunction_Sequence_1_ContainedGraph {nodes [
   n1 = RigVMFunctionEntryNode()
   RigVMFunction_Sequence = &RigVMFunction_Sequence(ExecuteContext=@n1.ExecuteContext)
   n2 = &RigVMFunction_Sequence(ExecuteContext=@RigVMFunction_Sequence.B)
   RigVMFunctionReturnNode(A=@RigVMFunction_Sequence.A, B=@n2.A, C=@n2.B)
  ]}
 RigVMModel {nodes [
   n1 = &RigUnit_PrepareForExecution()
   &RigUnit_HierarchyImportFromSkeleton(NameSpace=None, bIncludeCurves=False, bIncludeMeshSockets=False, bIncludeVirtualBones=True, ExecutePin=@n1.ExecutePin)
   RigUnit_BeginExecution = &RigUnit_BeginExecution()
   FN_DeformRig_FNC = RigVMFunctionReferenceNode(
    DeformRig_ Curve="(deformRig_l_upperarm_up_alpha_72_#1=False,deformRig_r_upperarm_up_alpha_73_#2=False,deformRig_l_upperarm_fr_alpha_74_#3=False,deformRig_r_upperarm_fr_alpha_75_CB51B8 ...(truncated)"
    DeformRig_Bool="(upperarm_1_#4=True,elbow_3_#5=True,wrist_7_#6=True,groin_8_#7=True,glute_17_#8=True,knee_9_40C408324CA261 ...(truncated)"
    DeformRig_Debug=(upperarm_debug_18_#4=False,wrist_debug_20_#6=False,lat_debug_26_#9=False,groinGlute_debug_27_#7=False)
    DeformRig_Value="(upperarm_up_18_#10=(Rotation=(X=0.0,Y=-0.0,Z=0.0,W=1.0),Translation=(X=0.0,Y=0.0,Z=8.0),Scale3D=(X=0.70,Y=1.0,Z=1.0)),upperarm_fwd_fr_19_#11=(Rotation=(X=0.0,Y=-0.0,Z=0.0,W=1.0),Translation=(X=0.0, ...(truncated)"
    ExecuteContext=@RigUnit_BeginExecution.ExecutePin
   )
   > wrist_function = RigVMFunctionReferenceNode()
   RigVMFunction_ControlFlowBranch = &RigVMFunction_ControlFlowBranch(Condition=RigVMVariableNode() {variable=isFrontEnd}.Value, ExecuteContext=@wrist_function.ExecuteContext)
   neckSetup = RigVMFunctionReferenceNode(ExecuteContext=@RigVMFunction_ControlFlowBranch.True)
   RigUnit_GetTransform = &RigUnit_GetTransform(Item=(Type=Control,Name="head_ctrl"), bInitial=False)
   RigUnit_SetTransform = &RigUnit_SetTransform(ExecutePin=(), Item=(Type=Bone,Name="head"), Weight=1, bInitial=false, Value=@RigUnit_GetTransform.Transform)
  ]}
]
hierarchy [head_ctrl {controlType=EULER_TRANSFORM, type=Control, loc=<0 0 0>, rot=<0 0 0 1>}]
variables [
 Head Rotation {type=FVector, default=(X=0,Y=0,Z=0)}
 isFrontEnd {type=bool, default=False}
]
