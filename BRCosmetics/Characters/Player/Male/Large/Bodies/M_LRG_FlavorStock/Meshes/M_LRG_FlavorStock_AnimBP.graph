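# Post-process style cosmetic AnimBP for the large male FlavorStock player body, copying the attached parent mesh's pose and adjusting hidden arms.
# The copied pose runs through the body's Control Rig, which receives the front-end flag.
# Hidden-arm bools then gate upper-arm scale and lower-arm twist offsets per side, with right upper-arm scale additive but left replacing.
# A second lowerarm/upperarm-twist shrink chain is disconnected from the output, and the EventGraph is empty.
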AnimBlueprint /BRCosmetics/Characters/Player/Male/Large/Bodies/M_LRG_FlavorStock/Meshes/M_LRG_FlavorStock_AnimBP
target_skeleton=&/M_LRG_FlavorStock_Skeleton
graphs [
 AnimGraph {nodes [
   n1 = Get bIsLeftArmHidden() -> (bIsLeftArmHidden:bool)
   n2 = "Transform (Modify) Bone - Bone: lowerarm_l"!(ComponentPose=Local To Component(), bAlphaBoolEnabled:bool=@n1, Translation="0,-6.212,-0.844", Rotation="0,0,0", Scale="0.250,0.250,0.250") -> (Pose) {
    alphaInputType=Bool
    scaleMode=BMM_Replace
    scaleSpace=BCS_BoneSpace
    translationMode=BMM_Additive
    translationSpace=BCS_BoneSpace
    alphaBoolBlend {blendOption=Linear}
    boneToModify {boneName=lowerarm_l}
    scale=<0.25 0.25 0.25>
    translation {y=-6.212, z=-0.844}
   }
   n3 = "Transform (Modify) Bone - Bone: upperarm_twist_02_l"!(ComponentPose=@n2, bAlphaBoolEnabled:bool=@n1, Translation="0,-1.5,-3", Rotation="0,0,0", Scale="0.500,0.500,0.500") -> (Pose) {
    alphaInputType=Bool
    scaleMode=BMM_Replace
    scaleSpace=BCS_BoneSpace
    translationMode=BMM_Additive
    translationSpace=BCS_BoneSpace
    alphaBoolBlend {blendOption=Linear}
    boneToModify {boneName=upperarm_twist_02_l}
    scale=<0.5 0.5 0.5>
    translation {y=-1.5, z=-3}
   }
   n4 = Get bIsRightArmHidden() -> (bIsRightArmHidden:bool)
   n5 = "Transform (Modify) Bone - Bone: lowerarm_r"!(ComponentPose=@n3, bAlphaBoolEnabled:bool=@n4, Translation="0,6.212,0.844", Rotation="0,0,0", Scale="0.25,0.25,0.25") -> (Pose) {
    alphaInputType=Bool
    scaleMode=BMM_Replace
    scaleSpace=BCS_BoneSpace
    translationMode=BMM_Additive
    translationSpace=BCS_BoneSpace
    alphaBoolBlend {blendOption=Linear}
    boneToModify {boneName=lowerarm_r}
    scale=<0.25 0.25 0.25>
    translation {y=6.212, z=0.844}
   }
   "Transform (Modify) Bone - Bone: upperarm_twist_02_r"!(ComponentPose=@n5, bAlphaBoolEnabled:bool=@n4, Translation="0,1.5,3", Rotation="0,0,0", Scale="0.5,0.5,0.5") -> (Pose) {
    alphaInputType=Bool
    scaleMode=BMM_Replace
    scaleSpace=BCS_BoneSpace
    translationMode=BMM_Additive
    translationSpace=BCS_BoneSpace
    alphaBoolBlend {blendOption=Linear}
    boneToModify {boneName=upperarm_twist_02_r}
    scale=<0.5 0.5 0.5>
    translation {y=1.5, z=3}
   }
   n6 = Control Rig!(Source=Copy Pose From Mesh() {bUseAttachedParent=true}, Alpha, isFrontEnd:bool=Get bIsInFrontEnd()) -> (Pose) {
    alphaBoolBlend {blendOption=Linear}
    controlRigAssetReference {blueprintRigClass=&/M_LRG_FlavorStock_CtrlRig.M_LRG_FlavorStock_CtrlRig_C}
   }
   n7 = Get bIsLeftArmHidden()
   n8 = "Transform (Modify) Bone - Bone: upperarm_l"!(ComponentPose=Local To Component(LocalPose=@n6), bAlphaBoolEnabled:bool=@n7, Translation="0,0,0", Rotation="0,0,0", Scale="1,0.9,1") -> (Pose) {
    alphaInputType=Bool
    scaleMode=BMM_Replace
    scaleSpace=BCS_BoneSpace
    translationMode=BMM_Additive
    translationSpace=BCS_BoneSpace
    alphaBoolBlend {blendOption=Linear}
    boneToModify {boneName=upperarm_l}
    scale {y=0.9}
   }
   n9 = "Transform (Modify) Bone - Bone: lowerarm_twist_02_l"!(ComponentPose=@n8, bAlphaBoolEnabled:bool=@n7, Translation="0,-3,-3", Rotation="0,0,0", Scale="1,1,1") -> (Pose) {
    alphaInputType=Bool
    scaleMode=BMM_Additive
    scaleSpace=BCS_BoneSpace
    translationMode=BMM_Additive
    translationSpace=BCS_BoneSpace
    alphaBoolBlend {blendOption=Linear}
    boneToModify {boneName=lowerarm_twist_02_l}
    translation {y=-3, z=-3}
   }
   n10 = Get bIsRightArmHidden()
   n11 = "Transform (Modify) Bone - Bone: upperarm_r"!(ComponentPose=@n9, bAlphaBoolEnabled:bool=@n10, Translation="0,0,0", Rotation="0,0,0", Scale="0.98,0.9,1") -> (Pose) {
    alphaInputType=Bool
    scaleMode=BMM_Additive
    scaleSpace=BCS_BoneSpace
    translationMode=BMM_Additive
    translationSpace=BCS_BoneSpace
    alphaBoolBlend {blendOption=Linear}
    boneToModify {boneName=upperarm_r}
    scale=<0.98 0.9>
   }
   n12 = "Transform (Modify) Bone - Bone: lowerarm_twist_02_r"!(ComponentPose=@n11, bAlphaBoolEnabled:bool=@n10, Translation="0,3,3", Rotation="0,0,0", Scale="1,1,1") -> (Pose) {
    alphaInputType=Bool
    scaleMode=BMM_Additive
    scaleSpace=BCS_BoneSpace
    translationMode=BMM_Additive
    translationSpace=BCS_BoneSpace
    alphaBoolBlend {blendOption=Linear}
    boneToModify {boneName=lowerarm_twist_02_r}
    translation {y=3, z=3}
   }
   Output Pose(Result=Component To Local(ComponentPose=@n12)) {layerGroup=None}
  ]}
 EventGraph {nodes [Event BlueprintUpdateAnimation()]}
]
variables []
