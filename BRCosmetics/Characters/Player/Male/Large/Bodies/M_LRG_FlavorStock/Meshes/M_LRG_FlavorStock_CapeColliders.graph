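# Physics asset for the large male FlavorStock player body, supplying collision shapes that, by its name, a cape collides against.
# Capsules cover the pelvis, spine, clavicles and neck, with a sphere on the head and extra capsules layered on spine_02 and spine_05.
# Arm, leg and foot bodies carry no shapes.
# There are no constraints, so the bodies form colliders only, not a jointed ragdoll.
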PhysicsAsset /BRCosmetics/Characters/Player/Male/Large/Bodies/M_LRG_FlavorStock/Meshes/M_LRG_FlavorStock_CapeColliders
bodies {
 calf_l []
 calf_r []
 clavicle_l [{shapeType=Capsule, length=15, radius=12, center=<9.8 -2.2 2.4>, rotation=<rot 69.5 0 0>}]
 clavicle_r [{shapeType=Capsule, length=15, radius=12, center=<-9.8 2.2 -2.4>, rotation=<rot 69 0 0>}]
 foot_l []
 foot_r []
 head [{shapeType=Sphere, radius=17.3125, center=<11.028 -0.174 0>}]
 lowerarm_l []
 lowerarm_r []
 neck_02 [{shapeType=Capsule, length=13.4375, radius=7.3175, center=<-0.2642 -0.3408 0>, rotation=<rot 90 -171.1193 -175.3978>}]
 pelvis [{shapeType=Capsule, length=14, radius=17, center=<-7.0228 -0.8063 0>}]
 spine_01 [{shapeType=Capsule, length=8, radius=19, center=<-1.7965 2.5607 0>}]
 spine_02 [
  {shapeType=Capsule, length=7.3264, radius=18, center=<-2.0573 0 0.0001>}
  {shapeType=Capsule, length=40, radius=13, center=<2.5024 0 3.9825>, rotation=<rot 90 0 5>}
  {shapeType=Capsule, length=40, radius=13, center=<2.4704 0 -3.9825>, rotation=<rot 90 0 5>}
 ]
 spine_03 [{shapeType=Capsule, length=11, radius=19, center=<0.1445 -1.128 0.0001>}]
 spine_04 [{shapeType=Capsule, length=11, radius=23, center=<2.766 0.7932 0.0002>}]
 spine_05 [{shapeType=Capsule, length=16, radius=20, center=<-2 -1.7928 0>}, {shapeType=Capsule, length=11, radius=15, center=<-5 -7 0>}]
 thigh_l []
 thigh_r []
 upperarm_l []
 upperarm_r []
}
constraints []
